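; Looping gameplay cue for the SunRose wings crash state, handling crash camera, wing material and audio on the target pawn.
; On loop start it adds a crashing camera-mode override, sets boost material parameters and flags forced crashes to audio from a gameplay tag.
; An ammo-change binding and repeating timer feed energy and speed to audio components.
; Winged-flight tag removal, view-target change or cue removal clears the camera override.
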(event EventOnApplication(EitherVFX-preferthisone)
  (params (MyTarget :type Actor%20Object%20Reference) (Parameters :type Gameplay%20Cue%20Parameters%20Structure%20%28by%20ref%29) (ParticleComponents :type Array%20of%20FXSystem%20Component%20Object%20References) (AudioComponents :type Array%20of%20Audio%20Component%20Object%20References) (BurstCameraShakeInstance :type Camera%20Shake%20Base%20Object%20Reference) (BurstDecalInstance :type Decal%20Actor%20Object%20Reference))
  (|CallParentFunction MyTarget Parameters ParticleComponents AudioComponents BurstCameraShakeInstance BurstDecalInstance)
  (bind _asfort_pawn (Utilities|Casting|CastToFortPawn MyTarget))
  (Variables|Default|SetisForcedCrash (GameplayTags|HasMatchingGameplayTag _asfort_pawn "(TagName=\"Granted.Athena.SunRose.Wings.ForceCrash.Damaged\")"))
  (for _array_element AudioComponents
    (Audio|Parameter|SetBooleanParameter _array_element "isForcedCrash" (Variables|Default|GetisForcedCrash) :!class /Script/AudioExtensions.AudioParameterControllerInterface)))

(event EventOnLoopingStart(EitherVFX-preferthisversion)
  (params (MyTarget :type Actor%20Object%20Reference) (Parameters :type Gameplay%20Cue%20Parameters%20Structure%20%28by%20ref%29) (ParticleComponents :type Array%20of%20FXSystem%20Component%20Object%20References) (AudioComponents :type Array%20of%20Audio%20Component%20Object%20References))
  (bind _fort_player_controller (Variables|Default|GetFort%20Player%20Controller))
  (bind _athenasunrosewings (Variables|TransientComponents|GetAthenaSunroseWings))
  (bind _materialinstance (Variables|/SunRoseFlyingGameplay/Items/Wings/Blueprints/B_Athena_SunRose_Wings.B_Athena_SunRose_Wings_C|GetMaterialInstance _athenasunrosewings))
  (|CallParentFunction MyTarget Parameters ParticleComponents AudioComponents)
  (bind _asfort_pawn (Utilities|Casting|CastToFortPawn MyTarget))
  (bind _output_get (Variables|TransientComponents|SetFortPawn _asfort_pawn))
  (Utilities|FlowControl|Sequence
    (:then_0
      (CallFunction|SetCameraEnabled _output_get true)
      (Utilities|FlowControl|Sequence
        (:then_0
          (Utilities|IsValid _fort_player_controller
            (:Is%20Valid
              (EventDispatchers|BindEventtoOnViewTargetChanged _fort_player_controller (EventDispatchers|CreateEvent :!event Custom|HandleViewTargetChanged)))))
        (:then_1
          (bind _asynctaskproxy (Ability|Tasks|WaitGameplayTagRemoveFromActor (Variables|TransientComponents|GetFortPawn) "(TagName=\"MovementMode.Flying.WingedFlight\")")
            (:then
              (Variables|Default|SetAsync%20Action_FlyMode _asynctaskproxy))
            (:Removed
              (CallFunction|SetCameraEnabled (Variables|TransientComponents|GetFortPawn)))))))
    (:then_1
      (bind _asb_athena_sun_rose_wings (Utilities|Casting|CastToB_Athena_SunRose_Wings (Variables|/Script/FortniteGame.FortPawn|GetCurrentWeapon _asfort_pawn)))
      (Variables|TransientComponents|SetAthenaSunroseWings _asb_athena_sun_rose_wings)
      (Variables|TransientComponents|SetAudioComponents AudioComponents)
      (Rendering|Material|SetScalarParameterValue _materialinstance "Boost" (Utilities|Time|GetGameTimeinSeconds (Variables|TransientComponents|GetFortPawn)) :!class /Script/Engine.MaterialInstanceDynamic)
      (Rendering|Material|SetScalarParameterValue _materialinstance "BoostSpeed" 1.4 :!class /Script/Engine.MaterialInstanceDynamic)
      (Rendering|Material|SetVectorParameterValue _materialinstance "BoostColor" "(R=1.000000,G=0.366660,B=0.103924,A=1.000000)" :!class /Script/Engine.MaterialInstanceDynamic)
      (EventDispatchers|BindEventtoOnLocalAmmoChanged (Variables|TransientComponents|GetAthenaSunroseWings) (delegate Custom|OnEnergyChange))
      (CallFunction|SetFXFloatParameter "EnergyRemaining" (Variables|/SunRoseFlyingGameplay/Items/Wings/Blueprints/B_Athena_SunRose_Wings.B_Athena_SunRose_Wings_C|GetPctEnergyRemaining (Variables|TransientComponents|GetAthenaSunroseWings)))
      (bind _returnvalue (Utilities|Time|SetTimerbyEvent (delegate Custom|SendSpeed) 0.1 true))
      (Variables|Default|SetTimer_IsFlapingCheck _returnvalue)
      (Variables|Default|SetisForcedCrash (GameplayTags|HasMatchingGameplayTag (Variables|TransientComponents|GetFortPawn) "(TagName=\"Granted.Athena.SunRose.Wings.ForceCrash.Damaged\")"))
      (for _array_element (Variables|TransientComponents|GetAudioComponents)
        (Audio|Parameter|SetBooleanParameter _array_element "isForcedCrash" (Variables|Default|GetisForcedCrash) :!class /Script/AudioExtensions.AudioParameterControllerInterface)))))

(event Custom|OnEnergyChange
  (params (LocalCount :type Integer) (LocalRemaining :type Integer))
  (CallFunction|SetFXFloatParameter "EnergyRemaining" (Variables|/SunRoseFlyingGameplay/Items/Wings/Blueprints/B_Athena_SunRose_Wings.B_Athena_SunRose_Wings_C|GetPctEnergyRemaining (Variables|TransientComponents|GetAthenaSunroseWings))))

(event Custom|SendSpeed
  (CallFunction|SetFXFloatParameter "Speed" (Math|Vector|VectorLength (Transformation|GetVelocity (Variables|TransientComponents|GetFortPawn)))))

(event EventOnRemoval(EitherVFX-preferthisversion)
  (params (MyTarget :type Actor%20Object%20Reference) (Parameters :type Gameplay%20Cue%20Parameters%20Structure%20%28by%20ref%29) (ParticleComponents :type Array%20of%20FXSystem%20Component%20Object%20References) (AudioComponents :type Array%20of%20Audio%20Component%20Object%20References) (BurstCameraShakeInstance :type Camera%20Shake%20Base%20Object%20Reference) (BurstDecalInstance :type Decal%20Actor%20Object%20Reference))
  (bind _fort_player_controller (Variables|Default|GetFort%20Player%20Controller))
  (|CallParentFunction MyTarget Parameters ParticleComponents AudioComponents BurstCameraShakeInstance BurstDecalInstance)
  (Utilities|FlowControl|Sequence
    (:then_0
      (CallFunction|SetCameraEnabled MyTarget)
      (Variables|Default|SetFortPlayerCameraBase 0)
      (Utilities|IsValid _fort_player_controller
        (:Is%20Valid
          (EventDispatchers|UnbindEventfromOnViewTargetChanged _fort_player_controller (EventDispatchers|CreateEvent :!event Custom|HandleViewTargetChanged))
          (Variables|Default|SetFort%20Player%20Controller 0))))
    (:then_1
      (Utilities|Time|ClearandInvalidateTimerbyHandle MyTarget (Variables|Default|GetTimer_IsFlapingCheck))
      (Ability|Async|EndAction (Variables|Default|GetAsync%20Action_FlyMode)))))

(event Custom|HandleViewTargetChanged
  (params (PlayerController :type Fort%20Player%20Controller%20Object%20Reference) (OldViewTarget :type Actor%20Object%20Reference) (NewViewTarget :type Actor%20Object%20Reference))
  (CallFunction|SetCameraEnabled OldViewTarget))

(fn SetCameraEnabled
  (params (TargetActor :type Actor%20Object%20Reference) (Enabled :type Boolean))
  (Variables|Default|GetFortPlayerCameraBase
    (:then
      _)
    (:else
      (bind _asfort_player_pawn (Utilities|Casting|CastToFortPlayerPawn TargetActor))
      (LocalVariables|SetFortPlayerPawn _asfort_player_pawn)
      (bind _asfort_player_controller (Utilities|Casting|CastToFortPlayerController (Pawn|GetLocalViewingPlayerController _asfort_player_pawn)))
      (Variables|Default|SetFort%20Player%20Controller _asfort_player_controller)
      (bind _asfort_player_camera_base (Utilities|Casting|CastToFortPlayerCameraBase (Variables|/Script/Engine.PlayerController|GetPlayerCameraManager _asfort_player_controller)))
      (Variables|Default|SetFortPlayerCameraBase _asfort_player_camera_base)))
  ;[1 1 1] Camera Mode override
  (if (Variables|GetEnabled)
    (Camera|AddCameraModeOverride (Variables|Default|GetFortPlayerCameraBase) (Variables|Camera|GetCameraModeClass_Crashing) (Variables|Camera|GetTag_CrashingCameraMode) 0 0 0)
    (else
      (Camera|RemoveCameraModeOverride (Variables|Default|GetFortPlayerCameraBase) (Variables|Camera|GetTag_CrashingCameraMode) 0 false))))

(fn SetFXFloatParameter
  (params (Name :type Name) (Float :type Float%20%28double-precision%29))
  (for _array_element (Variables|TransientComponents|GetAudioComponents)
    (Audio|Parameter|SetFloatParameter _array_element (Variables|GetName) (Variables|GetFloat) :!class /Script/AudioExtensions.AudioParameterControllerInterface)))

(fn ConstructionScript)
