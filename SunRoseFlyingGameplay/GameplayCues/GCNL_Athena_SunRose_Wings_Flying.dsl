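; Looping gameplay cue for SunRose Wings flight, driving its trail FX, audio and camera.
; On start it spawns replenish FX, activates a control bus mix and enables the flying camera mode and modifier, until flight ends or view target changes.
; A repeating timer feeds wing AnimBP state, skim checks and clamp-smoothed speed into FX and audio.
; A crash-tag watcher tracks crashing state, which selects the sound indicator settings shown to non-teammates.
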(event EventOnLoopingStart(EitherVFX-preferthisversion)
  (params (MyTarget :type Actor%20Object%20Reference) (Parameters :type Gameplay%20Cue%20Parameters%20Structure%20%28by%20ref%29) (ParticleComponents :type Array%20of%20FXSystem%20Component%20Object%20References) (AudioComponents :type Array%20of%20Audio%20Component%20Object%20References))
  (bind _playerpawn (Variables|Default|GetPlayerPawn))
  (bind _fort_player_controller (Variables|Default|GetFort%20Player%20Controller))
  (|CallParentFunction MyTarget Parameters ParticleComponents AudioComponents)
  (Variables|SpeedSmoothing|SetPreviousSpeed 0.0)
  (bind _asfort_player_pawn_athena (Utilities|Casting|CastToFortPlayerPawnAthena MyTarget))
  (Variables|Default|SetPlayerPawn _asfort_player_pawn_athena)
  (CallFunction|InitializeAudioSetup)
  (Utilities|FlowControl|Sequence
    (:then_0
      (bind (_asynctaskproxy _tagcount) (Ability|Tasks|WaitGameplayTagCountChangedOnActor (Variables|Default|GetPlayerPawn) (Variables|Default|GetTag_Crashing))
        (:then
          (Variables|AudioVisualizer|SetAction_WaitCrashTagChanged _asynctaskproxy)
          (Utilities|FlowControl|Sequence
            (:then_0
              (bind _asb_athena_sun_rose_wings (Utilities|Casting|CastToB_Athena_SunRose_Wings (Variables|/Script/FortniteGame.FortPawn|GetCurrentWeapon (Variables|Default|GetPlayerPawn))))
              (Variables|TransientComponents|SetAthenaSunroseWings _asb_athena_sun_rose_wings)
              (Variables|TransientComponents|SetAudioComponents AudioComponents)
              (Variables|TransientComponents|SetFlyingTrailFX (Utilities|Array|Get(acopy) ParticleComponents))
              ;[1 1 1] Replenish FX
              (Effects|SpawnFXSystemAttached :FXSystemAsset "/SunRoseFlyingGameplay/Items/Wings/FX/NS_SunRose_Wings_Replenish.NS_SunRose_Wings_Replenish" :AttachToComponent (Variables|/Script/FortniteGame.FortWeapon|GetWeaponMesh (Variables|TransientComponents|GetAthenaSunroseWings)) :AttachPointName "None" :Scale "1.000000,1.000000,1.000000" :LocationType "KeepRelativeOffset" :bAutoDestroy true :PoolingMethod "AutoRelease" :bAutoActivate true)
              (Rendering|Material|SetScalarParameterValue (Variables|/SunRoseFlyingGameplay/Items/Wings/Blueprints/B_Athena_SunRose_Wings.B_Athena_SunRose_Wings_C|GetMaterialInstance (Variables|TransientComponents|GetAthenaSunroseWings)) "Replenish" (Utilities|Time|GetGameTimeinSeconds (Variables|Default|GetPlayerPawn)) :!class /Script/Engine.MaterialInstanceDynamic)
              (bind _returnvalue (Utilities|Time|SetTimerbyEvent (delegate Custom|CE_IsFlapingCheck) 0.1 true))
              (Variables|Default|SetTimer_IsFlapingCheck _returnvalue)
              (bind _assun_rose_wings_anim_bp (Utilities|Casting|CastToSunRose_Wings_AnimBP (Components|SkeletalMesh|GetAnimInstance (Variables|/Script/FortniteGame.FortWeapon|GetWeaponMesh (Variables|TransientComponents|GetAthenaSunroseWings)))))
              (Variables|Default|SetAnimBP _assun_rose_wings_anim_bp)
              (Effects|Components|ParticleSystem|SetBoolParameter (Variables|TransientComponents|GetFlyingTrailFX) "IsLocalPlayer" (Pawn|IsLocallyControlled _playerpawn))
              (for _array_element (Variables|TransientComponents|GetAudioComponents)
                (Audio|Parameter|SetBooleanParameter _array_element "isLocalPlayer" (Pawn|IsLocallyViewed _playerpawn) :!class /Script/AudioExtensions.AudioParameterControllerInterface)))
            (:then_1
              (Variables|AudioVisualizer|SetIsCrashing (GameplayTags|HasMatchingGameplayTag (Variables|Default|GetPlayerPawn) (Variables|Default|GetTag_Crashing))))))
        (:TagCountChanged
          ;[1 1 1] AV Indicator
          (Variables|AudioVisualizer|SetIsCrashing (> _tagcount 0))))
      (if (not (Team|OnLocalPlayersTeam (Variables|Default|GetPlayerPawn) (Variables|Default|GetPlayerPawn)))
        (Variables|/Script/FortniteGame.FortSoundIndicatorComponent|SetDefaultIconOverride (select (Variables|AudioVisualizer|GetIsCrashing) (Variables|AudioVisualizer|GetIcon_AV_Crashing) (Variables|AudioVisualizer|GetIcon_AV_Default)) (Variables|Default|GetFortSoundIndicator))
        (Variables|/Script/FortniteGame.FortSoundIndicatorComponent|SetUpdateFrequency (Math|Float|SelectFloat (Variables|AudioVisualizer|GetAVUpdateRate_Crashing) (Variables|AudioVisualizer|GetAVUpdateRate_Default) (Variables|AudioVisualizer|GetIsCrashing)) (Variables|Default|GetFortSoundIndicator))
        (Variables|/Script/FortniteGame.FortSoundIndicatorComponent|SetDefaultSoundIndicatorType (select (Variables|AudioVisualizer|GetIsCrashing) (Variables|AudioVisualizer|GetAVIndicatorType_Crashing) (Variables|AudioVisualizer|GetAVIndicatorType_Default)) (Variables|Default|GetFortSoundIndicator))
        (Variables|/Script/FortniteGame.FortSoundIndicatorComponent|SetDefaultMaxVisibilityDistance (Math|Float|SelectFloat (ScalableFloat|GetValueatLevel (Variables|AudioVisualizer|GetRow_AVIndicatorRange_Crashing)) (ScalableFloat|GetValueatLevel (Variables|AudioVisualizer|GetRow_AVIndicatorRange_Default)) (Variables|AudioVisualizer|GetIsCrashing)) (Variables|Default|GetFortSoundIndicator))
        (SoundIndicator|StartTargetedSoundIndicatorLooping (Variables|Default|GetFortSoundIndicator) (Variables|Default|GetPlayerPawn) 0 1.0 (Variables|/Script/Engine.Pawn|GetPlayerState (Variables|Default|GetPlayerPawn)))))
    (:then_1
      (CallFunction|SetCameraEnabled (Variables|Default|GetPlayerPawn) true)
      (Utilities|FlowControl|Sequence
        (:then_0
          (Utilities|IsValid _fort_player_controller
            (:Is%20Valid
              (EventDispatchers|BindEventtoOnViewTargetChanged _fort_player_controller (EventDispatchers|CreateEvent :!event Custom|HandleViewTargetChanged)))))
        (:then_1
          (bind _asynctaskproxy_2 (Ability|Tasks|WaitGameplayTagRemoveFromActor (Variables|Default|GetPlayerPawn) (Variables|Default|GetTag_FlyingMovementMode))
            (:then
              (Variables|Default|SetAsync%20Action_FlyMode _asynctaskproxy_2))
            (:Removed
              (CallFunction|SetCameraEnabled (Variables|Default|GetPlayerPawn)))))))))

(event Custom|CE_IsFlapingCheck
  (CallFunction|SetFXAndAudioParams))

(event EventOnRemoval(EitherVFX-preferthisversion)
  (params (MyTarget :type Actor%20Object%20Reference) (Parameters :type Gameplay%20Cue%20Parameters%20Structure%20%28by%20ref%29) (ParticleComponents :type Array%20of%20FXSystem%20Component%20Object%20References) (AudioComponents :type Array%20of%20Audio%20Component%20Object%20References) (BurstCameraShakeInstance :type Camera%20Shake%20Base%20Object%20Reference) (BurstDecalInstance :type Decal%20Actor%20Object%20Reference))
  (bind _fort_player_controller (Variables|Default|GetFort%20Player%20Controller))
  (Utilities|FlowControl|Sequence
    (:then_0
      (CallFunction|SetCameraEnabled MyTarget)
      (Variables|Default|SetFortPlayerCameraBase 0)
      (Utilities|IsValid _fort_player_controller
        (:Is%20Valid
          (EventDispatchers|UnbindEventfromOnViewTargetChanged _fort_player_controller (EventDispatchers|CreateEvent :!event Custom|HandleViewTargetChanged))
          (Variables|Default|SetFort%20Player%20Controller 0))))
    (:then_1
      (Utilities|Time|ClearandInvalidateTimerbyHandle MyTarget (Variables|Default|GetTimer_IsFlapingCheck))
      (CallFunction|HandleControlBusMix false (Variables|Default|GetPlayerPawn))
      (SoundIndicator|StopSoundIndicatorLooping (Variables|Default|GetFortSoundIndicator))
      (Ability|Async|EndAction (Variables|AudioVisualizer|GetAction_WaitCrashTagChanged)))))

(event Custom|HandleViewTargetChanged
  (params (PlayerController :type Fort%20Player%20Controller%20Object%20Reference) (OldViewTarget :type Actor%20Object%20Reference) (NewViewTarget :type Actor%20Object%20Reference))
  (CallFunction|SetCameraEnabled OldViewTarget))

(fn GetClampedSpeed
  (params (ActorComponent :type Actor%20Component%20Object%20Reference))
  (return (Math|Float|Clamp(Float) (Math|Vector|VectorLength (Transformation|GetVelocity (Components|GetOwner ActorComponent))) (- (Variables|SpeedSmoothing|GetPreviousSpeed) 1.5e+02) (+ (Variables|SpeedSmoothing|GetPreviousSpeed) 2e+02))))

(fn HandleControlBusMix
  (params (Apply :type Boolean) (PlayerPawn :type Fort%20Player%20Pawn%20Object%20Reference))
  (bind _playerpawn (Variables|Default|GetPlayerPawn))
  (if (Pawn|IsLocallyViewed PlayerPawn)
    (if (Variables|GetApply)
      (Audio|Modulation|ActivateControlBusMix _playerpawn "/SunRoseFlyingGameplay/Sounds/CBM_Sunrose_Flying.CBM_Sunrose_Flying")
      (else
        (Audio|Modulation|DeactivateControlBusMix _playerpawn "/SunRoseFlyingGameplay/Sounds/CBM_Sunrose_Flying.CBM_Sunrose_Flying")))))

(fn HandleControlBusMix_0
  (params (Apply :type Boolean) (PlayerPawn :type Fort%20Player%20Pawn%20Object%20Reference))
  (bind _playerpawn (Variables|Default|GetPlayerPawn))
  (if (Pawn|IsLocallyViewed PlayerPawn)
    (if (Variables|GetApply)
      (Audio|Modulation|ActivateControlBusMix _playerpawn "/SunRoseFlyingGameplay/Sounds/CBM_Sunrose_Flying.CBM_Sunrose_Flying")
      (else
        (Audio|Modulation|DeactivateControlBusMix _playerpawn "/SunRoseFlyingGameplay/Sounds/CBM_Sunrose_Flying.CBM_Sunrose_Flying")))))

(fn InitializeAudioSetup
  (bind _playerpawn (Variables|Default|GetPlayerPawn))
  (bind _returnvalue (Actor|GetComponentbyClass _playerpawn "/SunRoseFlyingGameplay/Items/Wings/Blueprints/FlyingPawnComponent.FlyingPawnComponent_C"))
  (CallFunction|HandleControlBusMix true (Variables|Default|GetPlayerPawn))
  (Utilities|IsValid _returnvalue
    (:Is%20Valid
      (Variables|Default|SetFlyingPawnComponent _returnvalue))))

(fn PassSFXParameters
  (bind _audiocomponent (LocalVariables|GetAudioComponent))
  (bind _audiocomponent_2 (LocalVariables|GetAudioComponent))
  (for _array_element (Variables|TransientComponents|GetAudioComponents)
    (bind _output_get (LocalVariables|SetAudioComponent _array_element))
    (Utilities|IsValid _output_get
      (:Is%20Valid
        (Audio|Parameter|SetBooleanParameter _audiocomponent "isAscending" (Variables|/Script/SunRoseFlyingRuntime.SunRoseFlying_WingsAnimInstance|GetbPlayerIsAscending (Variables|Default|GetAnimBP)) :!class /Script/AudioExtensions.AudioParameterControllerInterface)
        (Audio|Parameter|SetBooleanParameter _audiocomponent "isBoosting" (Variables|/Script/SunRoseFlyingRuntime.SunRoseFlying_WingsAnimInstance|GetbPlayerIsWingsBoosting (Variables|Default|GetAnimBP)) :!class /Script/AudioExtensions.AudioParameterControllerInterface)
        (Audio|Parameter|SetBooleanParameter _audiocomponent "isDiving" (Variables|/Script/SunRoseFlyingRuntime.SunRoseFlying_WingsAnimInstance|GetbPlayerIsWingsCrashing (Variables|Default|GetAnimBP)) :!class /Script/AudioExtensions.AudioParameterControllerInterface)
        (Audio|Parameter|SetBooleanParameter _audiocomponent "isDodging" (or (Variables|/Script/SunRoseFlyingRuntime.SunRoseFlying_WingsAnimInstance|GetbPlayerIsDodgeLeft (Variables|Default|GetAnimBP)) (Variables|/Script/SunRoseFlyingRuntime.SunRoseFlying_WingsAnimInstance|GetbPlayerIsDodgeRight (Variables|Default|GetAnimBP))) :!class /Script/AudioExtensions.AudioParameterControllerInterface)
        (Audio|Parameter|SetFloatParameter _audiocomponent_2 "Pitch" (Variables|/Script/SunRoseFlyingRuntime.SunRoseFlying_WingsAnimInstance|GetPlayerGlidePitchAlpha (Variables|Default|GetAnimBP)) :!class /Script/AudioExtensions.AudioParameterControllerInterface)
        (Audio|Parameter|SetFloatParameter _audiocomponent_2 "Roll" (Variables|/Script/SunRoseFlyingRuntime.SunRoseFlying_WingsAnimInstance|GetPlayerGlideRollCombined (Variables|Default|GetAnimBP)) :!class /Script/AudioExtensions.AudioParameterControllerInterface)
        (bind _output_get_2 (Variables|SpeedSmoothing|SetPreviousSpeed (CallFunction|GetClampedSpeed _audiocomponent_2)))
        (Audio|Parameter|SetFloatParameter _audiocomponent_2 "Speed" _output_get_2 :!class /Script/AudioExtensions.AudioParameterControllerInterface)
        (Variables|Default|GetFlyingPawnComponent
          (:then
            (Audio|Parameter|SetBooleanParameter (LocalVariables|GetAudioComponent) "isSkimming" (or (>= (Variables|/SunRoseFlyingGameplay/Items/Wings/Blueprints/FlyingPawnComponent.FlyingPawnComponent_C|GetSkim_Distance_Left (Variables|Default|GetFlyingPawnComponent)) 0) (>= (Variables|/SunRoseFlyingGameplay/Items/Wings/Blueprints/FlyingPawnComponent.FlyingPawnComponent_C|GetSkim_Distance_Right (Variables|Default|GetFlyingPawnComponent)) 0)) :!class /Script/AudioExtensions.AudioParameterControllerInterface))
          (:else))))))

(fn SetFXandAudioParams
  (Variables|Default|SetAnimBP (Utilities|Casting|CastToSunRose_Wings_AnimBP (Components|SkeletalMesh|GetAnimInstance (Variables|/Script/FortniteGame.FortWeapon|GetWeaponMesh (Variables|TransientComponents|GetAthenaSunroseWings)))))
  (Variables|SpeedSmoothing|SetPreviousSpeed (CallFunction|GetClampedSpeed (Variables|/Script/Engine.Actor|GetRootComponent (Variables|Default|GetPlayerPawn))))
  (CallFunction|SetFlappingParameters (Variables|TransientComponents|GetFlyingTrailFX) (Variables|TransientComponents|GetAudioComponents) (Variables|Default|GetPlayerPawn) (Variables|/Script/SunRoseFlyingRuntime.SunRoseFlying_WingsAnimInstance|GetWingFlapAlpha (Variables|Default|GetAnimBP)) (Variables|/Script/SunRoseFlyingRuntime.SunRoseFlying_WingsAnimInstance|GetPlayerGlideRollCombined (Variables|Default|GetAnimBP)) (- 1.0 (Variables|/Script/SunRoseFlyingRuntime.SunRoseFlying_WingsAnimInstance|GetPlayerRollAlpha (Variables|Default|GetAnimBP))) (Variables|/Script/SunRoseFlyingRuntime.SunRoseFlying_WingsAnimInstance|GetPlayerGlidePitchAlpha (Variables|Default|GetAnimBP)) (Variables|/Script/SunRoseFlyingRuntime.SunRoseFlying_WingsAnimInstance|GetbPlayerIsWingsBoosting (Variables|Default|GetAnimBP)) (Math|Boolean|ORBoolean (Variables|/Script/SunRoseFlyingRuntime.SunRoseFlying_WingsAnimInstance|GetbPlayerIsAscending (Variables|Default|GetAnimBP)) (Variables|/Script/SunRoseFlyingRuntime.SunRoseFlying_WingsAnimInstance|GetbPlayerIsDodgeLeft (Variables|Default|GetAnimBP)) (Variables|/Script/SunRoseFlyingRuntime.SunRoseFlying_WingsAnimInstance|GetbPlayerIsDodgeRight (Variables|Default|GetAnimBP))) (Variables|/Script/SunRoseFlyingRuntime.SunRoseFlying_WingsAnimInstance|GetbPlayerIsAllowWingFlap (Variables|Default|GetAnimBP)) (or (>= (Variables|/SunRoseFlyingGameplay/Items/Wings/Blueprints/FlyingPawnComponent.FlyingPawnComponent_C|GetSkim_Distance_Left (Variables|Default|GetFlyingPawnComponent)) 0) (>= (Variables|/SunRoseFlyingGameplay/Items/Wings/Blueprints/FlyingPawnComponent.FlyingPawnComponent_C|GetSkim_Distance_Right (Variables|Default|GetFlyingPawnComponent)) 0)) (Variables|SpeedSmoothing|GetPreviousSpeed) (Variables|/Script/SunRoseFlyingRuntime.SunRoseFlying_WingsAnimInstance|GetbPlayerIsWingsCrashing (Variables|Default|GetAnimBP)) (Variables|/Script/SunRoseFlyingRuntime.FortPawnComponent_WingedFlying|GetNativeCurrentHeightAboveWorld (Variables|Default|GetFlyingPawnComponent))))

(fn SetCameraEnabled
  (params (TargetActor :type Actor%20Object%20Reference) (Enabled :type Boolean))
  (Variables|Default|GetFortPlayerCameraBase
    (:then
      _)
    (:else
      (bind _asfort_player_pawn (Utilities|Casting|CastToFortPlayerPawn TargetActor))
      (LocalVariables|SetFortPlayerPawn _asfort_player_pawn)
      (bind _asfort_player_controller (Utilities|Casting|CastToFortPlayerController (Pawn|GetLocalViewingPlayerController _asfort_player_pawn)))
      (Variables|Default|SetFort%20Player%20Controller _asfort_player_controller)
      (bind _asfort_player_camera_base (Utilities|Casting|CastToFortPlayerCameraBase (Variables|/Script/Engine.PlayerController|GetPlayerCameraManager _asfort_player_controller)))
      (Variables|Default|SetFortPlayerCameraBase _asfort_player_camera_base)))
  (Utilities|FlowControl|Sequence
    (:then_0
      ;[1 1 1] Camera Mode override
      (if (Variables|GetEnabled)
        (if (not (GameplayTags|HasAnyMatchingGameplayTags (Variables|Default|GetPlayerPawn) (Variables|Camera|GetTags_PreventApplyFlyingCameraMode)))
          (Camera|AddCameraModeOverride (Variables|Default|GetFortPlayerCameraBase) (Variables|Camera|GetCameraModeClass_Flying) (Variables|Camera|GetTag_FlyingCameraMode) 0 0 0))
        (else
          (Camera|RemoveCameraModeOverride (Variables|Default|GetFortPlayerCameraBase) (Variables|Camera|GetTag_FlyingCameraMode) 0 false))))
    (:then_1
      (if (Variables|GetEnabled)
        ;[1 1 1] Enable Camera Modifier
        (bind _returnvalue (Game|Player|FindCameraModifierbyClass (Variables|Default|GetFortPlayerCameraBase) (Variables|Camera|GetCameraModifierClass_Flying)))
        (Utilities|IsValid _returnvalue
          (:Is%20Valid
            (CameraModifier|EnableModifier _returnvalue))
          (:Is%20Not%20Valid
            (Game|Player|AddNewCameraModifier (Variables|Default|GetFortPlayerCameraBase) (Variables|Camera|GetCameraModifierClass_Flying))))
        (else
          ;[1 1 1] Disable Camera Modifier
          (bind _returnvalue_2 (Game|Player|FindCameraModifierbyClass (Variables|Default|GetFortPlayerCameraBase) (Variables|Camera|GetCameraModifierClass_Flying)))
          (Utilities|IsValid _returnvalue_2
            (:Is%20Valid
              (CameraModifier|DisableModifier _returnvalue_2 false))))))))

(fn SetFXFloatParameter
  (params (Name :type Name) (Float :type Float%20%28double-precision%29))
  (bind _flyingtrailfx (Variables|TransientComponents|GetFlyingTrailFX)
    (:then
      (Effects|Components|ParticleSystem|SetFloatParameter _flyingtrailfx (Variables|GetName) (Variables|GetFloat)))
    (:else
      _))
  (for _array_element (Variables|TransientComponents|GetAudioComponents)
    (Audio|Parameter|SetFloatParameter _array_element (Variables|GetName) (Variables|GetFloat) :!class /Script/AudioExtensions.AudioParameterControllerInterface)))

(fn ConstructionScript)
