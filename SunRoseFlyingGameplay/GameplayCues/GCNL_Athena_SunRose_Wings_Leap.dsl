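; Looping gameplay cue for the Leap ability of the SunRose wings, providing visual and audio feedback on the flying pawn.
; On start it shows a sound indicator for non-local pawns, stamps game time and boost color and speed into the wings' dynamic material, and sends remaining energy to audio.
; A repeating timer feeds pawn velocity into each audio component's Speed parameter through a shared helper, and removal clears that timer.
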(event EventOnLoopingStart(EitherVFX-preferthisversion)
  (params (MyTarget :type Actor%20Object%20Reference) (Parameters :type Gameplay%20Cue%20Parameters%20Structure%20%28by%20ref%29) (ParticleComponents :type Array%20of%20FXSystem%20Component%20Object%20References) (AudioComponents :type Array%20of%20Audio%20Component%20Object%20References))
  (bind _fortpawn (Variables|Default|GetFortPawn))
  (bind _asfort_pawn (Utilities|Casting|CastToFortPawn MyTarget))
  (Variables|Default|SetFortPawn _asfort_pawn)
  (if (not (Pawn|IsLocallyControlled (Variables|Default|GetFortPawn)))
    (Fortnite|ShowSoundIndicatorLocal :WorldContextObject (Variables|Default|GetFortPawn) :TrackedActor (Variables|Default|GetFortPawn) :Location (Transformation|GetActorLocation (Variables|Default|GetFortPawn)) :MaxAudibleDistance (ScalableFloat|GetValueatLevel (Variables|Default|GetRow_AVIndicatorDistance_Leap)) :IndicatorType "Gunshot" :Instigator (Variables|Default|GetFortPawn) :Tint "(R=0.000000,G=0.000000,B=0.000000,A=0.000000)" :OverrideIcon (Variables|Default|GetAVIcon) :AttenuationMaxRangeMultiplier 1.0 :AttenuationIntensityMultiplier 1.0))
  (bind _asb_athena_sun_rose_wings (Utilities|Casting|CastToB_Athena_SunRose_Wings (Variables|/Script/FortniteGame.FortPawn|GetCurrentWeapon _fortpawn)))
  (bind _output_get (Variables|Default|SetAthenaSunroseWings _asb_athena_sun_rose_wings))
  (Rendering|Material|SetScalarParameterValue (Variables|/SunRoseFlyingGameplay/Items/Wings/Blueprints/B_Athena_SunRose_Wings.B_Athena_SunRose_Wings_C|GetMaterialInstance _output_get) "Boost" (Utilities|Time|GetGameTimeinSeconds _fortpawn) :!class /Script/Engine.MaterialInstanceDynamic)
  (Rendering|Material|SetScalarParameterValue (Variables|/SunRoseFlyingGameplay/Items/Wings/Blueprints/B_Athena_SunRose_Wings.B_Athena_SunRose_Wings_C|GetMaterialInstance _output_get) "BoostSpeed" 0.7 :!class /Script/Engine.MaterialInstanceDynamic)
  (Rendering|Material|SetVectorParameterValue (Variables|/SunRoseFlyingGameplay/Items/Wings/Blueprints/B_Athena_SunRose_Wings.B_Athena_SunRose_Wings_C|GetMaterialInstance _output_get) "BoostColor" "(R=0.729167,G=0.304986,B=0.075955,A=1.000000)" :!class /Script/Engine.MaterialInstanceDynamic)
  (Variables|Default|SetAudioComponents AudioComponents)
  (CallFunction|SetAudioFloatParameter "EnergyRemaining" (Variables|/SunRoseFlyingGameplay/Items/Wings/Blueprints/B_Athena_SunRose_Wings.B_Athena_SunRose_Wings_C|GetPctEnergyRemaining (Variables|Default|GetAthenaSunroseWings)))
  (bind _returnvalue (Utilities|Time|SetTimerbyEvent (delegate Custom|GetSpeed) 0.1 true))
  (Variables|Default|SetTimer_IsFlapingCheck _returnvalue))

(event Custom|GetSpeed
  (CallFunction|SetAudioFloatParameter "Speed" (Math|Vector|VectorLength (Transformation|GetVelocity (Variables|Default|GetFortPawn)))))

(event EventOnRemoval(EitherVFX-preferthisversion)
  (params (MyTarget :type Actor%20Object%20Reference) (Parameters :type Gameplay%20Cue%20Parameters%20Structure%20%28by%20ref%29) (ParticleComponents :type Array%20of%20FXSystem%20Component%20Object%20References) (AudioComponents :type Array%20of%20Audio%20Component%20Object%20References) (BurstCameraShakeInstance :type Camera%20Shake%20Base%20Object%20Reference) (BurstDecalInstance :type Decal%20Actor%20Object%20Reference))
  (|CallParentFunction MyTarget Parameters ParticleComponents AudioComponents BurstCameraShakeInstance BurstDecalInstance)
  (Utilities|Time|ClearandInvalidateTimerbyHandle MyTarget (Variables|Default|GetTimer_IsFlapingCheck)))

(fn SetAudioFloatParameter
  (params (Name :type Name) (Float :type Float%20%28double-precision%29))
  (for _array_element (Variables|Default|GetAudioComponents)
    (Audio|Parameter|SetFloatParameter _array_element Name Float :!class /Script/AudioExtensions.AudioParameterControllerInterface)))

(fn ConstructionScript)
