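; Looping gameplay cue that drives skimming effects on SunRose flying wings while the player pawn skims a surface.
; On start it caches the skim particle, weapon mesh and FlyingPawnComponent, then a repeating timer passes each wing's skim location, normal and distance to ShowWingFX.
; Mobile, Switch and PC performance mode skip the timer and switch the particle's IsSkimming off.
; Outside those platforms, removal clears the timer and stops the skim audio.
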(event EventOnLoopingStart(EitherVFX-preferthisversion)
  (params (MyTarget :type Actor%20Object%20Reference) (Parameters :type Gameplay%20Cue%20Parameters%20Structure%20%28by%20ref%29) (ParticleComponents :type Array%20of%20FXSystem%20Component%20Object%20References) (AudioComponents :type Array%20of%20Audio%20Component%20Object%20References))
  (|CallParentFunction MyTarget Parameters ParticleComponents AudioComponents)
  (Variables|FortGameplayCueNotifyLoop_WingsSkimming|SetSkimFX (Utilities|Array|Get(acopy) ParticleComponents))
  (bind _asfort_player_pawn (Utilities|Casting|CastToFortPlayerPawn MyTarget))
  (Variables|Default|SetPlayerPawn _asfort_player_pawn)
  (Variables|Default|SetFlyingPawnComponent (Actor|GetComponentbyClass (Variables|Default|GetPlayerPawn) "/SunRoseFlyingGameplay/Items/Wings/Blueprints/FlyingPawnComponent.FlyingPawnComponent_C"))
  (Variables|FortGameplayCueNotifyLoop_WingsSkimming|SetWingsSkeletalMesh (Variables|/Script/FortniteGame.FortWeapon|GetWeaponMesh (Variables|/Script/FortniteGame.FortPawn|GetCurrentWeapon (Variables|Default|GetPlayerPawn))))
  (if (Math|Boolean|ORBoolean (Fortnite|IsMobilePlatform) (Class|PlatformHelpersLibrary|IsPcPerfMode (Variables|Default|GetPlayerPawn)) (TouchPlatform|IsMobileGame (Variables|Default|GetPlayerPawn)) (TouchPlatform|IsNintendoSwitchorNintendoSwitchPreview))
    (Effects|Components|ParticleSystem|SetBoolParameter (Variables|FortGameplayCueNotifyLoop_WingsSkimming|GetSkimFX) "IsSkimming" false)
    (else
      (bind _returnvalue (Utilities|Time|SetTimerbyEvent (delegate Custom|CheckTouchPoint) 0.1 true))
      (Variables|Default|SetTimer _returnvalue)
      ;[1 1 1] If skim distance is > 0, show FX for that wing
      (CallFunction|ShowWingFX (Variables|/SunRoseFlyingGameplay/Items/Wings/Blueprints/FlyingPawnComponent.FlyingPawnComponent_C|GetSkim_Location_Left (Variables|Default|GetFlyingPawnComponent)) (Variables|/SunRoseFlyingGameplay/Items/Wings/Blueprints/FlyingPawnComponent.FlyingPawnComponent_C|GetSkim_Normal_Left (Variables|Default|GetFlyingPawnComponent)) (Variables|/SunRoseFlyingGameplay/Items/Wings/Blueprints/FlyingPawnComponent.FlyingPawnComponent_C|GetSkim_Distance_Left (Variables|Default|GetFlyingPawnComponent)) (Variables|/SunRoseFlyingGameplay/Items/Wings/Blueprints/FlyingPawnComponent.FlyingPawnComponent_C|GetSkim_Location_Right (Variables|Default|GetFlyingPawnComponent)) (Variables|/SunRoseFlyingGameplay/Items/Wings/Blueprints/FlyingPawnComponent.FlyingPawnComponent_C|GetSkim_Normal_Right (Variables|Default|GetFlyingPawnComponent)) (Variables|/SunRoseFlyingGameplay/Items/Wings/Blueprints/FlyingPawnComponent.FlyingPawnComponent_C|GetSkim_Distance_Right (Variables|Default|GetFlyingPawnComponent))))))

(event Custom|CheckTouchPoint
  ;[1 1 1] If skim distance is > 0, show FX for that wing
  (CallFunction|ShowWingFX (Variables|/SunRoseFlyingGameplay/Items/Wings/Blueprints/FlyingPawnComponent.FlyingPawnComponent_C|GetSkim_Location_Left (Variables|Default|GetFlyingPawnComponent)) (Variables|/SunRoseFlyingGameplay/Items/Wings/Blueprints/FlyingPawnComponent.FlyingPawnComponent_C|GetSkim_Normal_Left (Variables|Default|GetFlyingPawnComponent)) (Variables|/SunRoseFlyingGameplay/Items/Wings/Blueprints/FlyingPawnComponent.FlyingPawnComponent_C|GetSkim_Distance_Left (Variables|Default|GetFlyingPawnComponent)) (Variables|/SunRoseFlyingGameplay/Items/Wings/Blueprints/FlyingPawnComponent.FlyingPawnComponent_C|GetSkim_Location_Right (Variables|Default|GetFlyingPawnComponent)) (Variables|/SunRoseFlyingGameplay/Items/Wings/Blueprints/FlyingPawnComponent.FlyingPawnComponent_C|GetSkim_Normal_Right (Variables|Default|GetFlyingPawnComponent)) (Variables|/SunRoseFlyingGameplay/Items/Wings/Blueprints/FlyingPawnComponent.FlyingPawnComponent_C|GetSkim_Distance_Right (Variables|Default|GetFlyingPawnComponent))))

(event EventOnRemoval(EitherVFX-preferthisversion)
  (params (MyTarget :type Actor%20Object%20Reference) (Parameters :type Gameplay%20Cue%20Parameters%20Structure%20%28by%20ref%29) (ParticleComponents :type Array%20of%20FXSystem%20Component%20Object%20References) (AudioComponents :type Array%20of%20Audio%20Component%20Object%20References) (BurstCameraShakeInstance :type Camera%20Shake%20Base%20Object%20Reference) (BurstDecalInstance :type Decal%20Actor%20Object%20Reference))
  (bind _skimaudiocomponent (Variables|FortGameplayCueNotifyLoop_WingsSkimming|GetSkimAudioComponent))
  (|CallParentFunction MyTarget Parameters ParticleComponents AudioComponents BurstCameraShakeInstance BurstDecalInstance)
  (if (not (Math|Boolean|ORBoolean (Fortnite|IsMobilePlatform) (Class|PlatformHelpersLibrary|IsPcPerfMode (Variables|Default|GetPlayerPawn)) (TouchPlatform|IsMobileGame (Variables|Default|GetPlayerPawn)) (TouchPlatform|IsNintendoSwitchorNintendoSwitchPreview)))
    (Utilities|Time|ClearandInvalidateTimerbyHandle 0 (Variables|Default|GetTimer))
    (Utilities|IsValid _skimaudiocomponent
      (:Is%20Valid
        (Audio|Components|Audio|Stop _skimaudiocomponent)))))

(event Custom|DelayStopSkimmingAudio
  (bind _skimaudiocomponent (Variables|FortGameplayCueNotifyLoop_WingsSkimming|GetSkimAudioComponent))
  (bind _skimaudiocomponent_2 (Variables|FortGameplayCueNotifyLoop_WingsSkimming|GetSkimAudioComponent))
  (Utilities|IsValid _skimaudiocomponent
    (:Is%20Valid
      (Utilities|FlowControl|RetriggerableDelay _skimaudiocomponent 2.0)
      (Utilities|IsValid _skimaudiocomponent_2
        (:Is%20Valid
          (Audio|Components|Audio|Stop _skimaudiocomponent_2))))))

(fn ConstructionScript)
