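; Looping gameplay cue for the SunRose flying wings' speed threshold, driving its particle and audio effects from the pawn and its equipped wings.
; On start it caches components, flags local control on the first particle system and sends it initial remaining energy alongside audio.
; Afterward, the wings' ammo-change event updates energy and a repeating timer samples speed, both for audio only.
; Removal clears the timer but leaves the ammo event bound.
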(event EventOnLoopingStart(EitherVFX-preferthisversion)
  (params (MyTarget :type Actor%20Object%20Reference) (Parameters :type Gameplay%20Cue%20Parameters%20Structure%20%28by%20ref%29) (ParticleComponents :type Array%20of%20FXSystem%20Component%20Object%20References) (AudioComponents :type Array%20of%20Audio%20Component%20Object%20References))
  (bind _output (Utilities|Array|Get(acopy) ParticleComponents))
  (|CallParentFunction MyTarget Parameters ParticleComponents AudioComponents)
  (bind _asfort_pawn (Utilities|Casting|CastToFortPawn MyTarget))
  (Variables|Default|SetFortPawn _asfort_pawn)
  (bind _asb_athena_sun_rose_wings (Utilities|Casting|CastToB_Athena_SunRose_Wings (Variables|/Script/FortniteGame.FortPawn|GetCurrentWeapon _asfort_pawn)))
  (Variables|Default|SetAthenaSunroseWings _asb_athena_sun_rose_wings)
  (Variables|Default|SetAudioComponents AudioComponents)
  (Variables|Default|SetParticleComponent _output)
  (Effects|Components|ParticleSystem|SetBoolParameter _output "IsLocalPlayer" (Pawn|IsLocallyControlled (Variables|Default|GetFortPawn)))
  (EventDispatchers|BindEventtoOnLocalAmmoChanged (Variables|Default|GetAthenaSunroseWings) (delegate Custom|OnEnergyChange))
  (Default|SetFXFloatParameter "EnergyRemaining" (Variables|/SunRoseFlyingGameplay/Items/Wings/Blueprints/B_Athena_SunRose_Wings.B_Athena_SunRose_Wings_C|GetPctEnergyRemaining (Variables|Default|GetAthenaSunroseWings)))
  (bind _returnvalue (Utilities|Time|SetTimerbyEvent (delegate Custom|GetSpeed) 0.1 true))
  (Variables|Default|SetTimer_SFXParamsCheck _returnvalue))

(event Custom|OnEnergyChange
  (params (LocalCount :type Integer) (LocalRemaining :type Integer))
  (Default|SetFXFloatParameter "EnergyRemaining" (Variables|/SunRoseFlyingGameplay/Items/Wings/Blueprints/B_Athena_SunRose_Wings.B_Athena_SunRose_Wings_C|GetPctEnergyRemaining (Variables|Default|GetAthenaSunroseWings)) false))

(event Custom|GetSpeed
  (Default|SetFXFloatParameter "Speed" (Math|Vector|VectorLength (Transformation|GetVelocity (Variables|Default|GetFortPawn))) false))

(event EventOnRemoval(EitherVFX-preferthisversion)
  (params (MyTarget :type Actor%20Object%20Reference) (Parameters :type Gameplay%20Cue%20Parameters%20Structure%20%28by%20ref%29) (ParticleComponents :type Array%20of%20FXSystem%20Component%20Object%20References) (AudioComponents :type Array%20of%20Audio%20Component%20Object%20References) (BurstCameraShakeInstance :type Camera%20Shake%20Base%20Object%20Reference) (BurstDecalInstance :type Decal%20Actor%20Object%20Reference))
  (|CallParentFunction MyTarget Parameters ParticleComponents AudioComponents BurstCameraShakeInstance BurstDecalInstance)
  (Utilities|Time|ClearandInvalidateTimerbyHandle MyTarget (Variables|Default|GetTimer_SFXParamsCheck)))

(fn SetFXFloatParameter
  (params (Name :type Name) (Float :type Float%20%28double-precision%29) (SendToVFX :type Boolean :default true))
  (if (and (Utilities|IsValid (Variables|Default|GetParticleComponent) :!class /Script/Engine.KismetSystemLibrary) (Variables|GetSendToVFX))
    (Effects|Components|ParticleSystem|SetFloatParameter (Variables|Default|GetParticleComponent) (Variables|GetName) (Variables|GetFloat))
    (else
      _))
  (for _array_element (Variables|Default|GetAudioComponents)
    (Audio|Parameter|SetFloatParameter _array_element (Variables|GetName) (Variables|GetFloat) :!class /Script/AudioExtensions.AudioParameterControllerInterface)))

(fn ConstructionScript)
