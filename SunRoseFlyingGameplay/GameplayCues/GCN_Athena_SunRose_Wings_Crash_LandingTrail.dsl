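; Gameplay cue notify for the SunRose wings crash-landing trail, feeding player and altitude data to its spawned effect on burst.
; OnBurst casts the target to a Fort player pawn and reads height above world from its flying component.
; The first particle component receives an IsLocalPlayer flag and an AltitudeInterpolate float, the height remapped to a clamped 0–1 range.
; A leftover debug print of that value has screen and log output disabled.
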(fn OnBurst(EitherVFX-preferthisversion)
  (params (MyTarget :type Actor%20Object%20Reference) (Parameters :type Gameplay%20Cue%20Parameters%20Structure%20%28by%20ref%29) (ParticleComponents :type Array%20of%20FXSystem%20Component%20Object%20References) (AudioComponents :type Array%20of%20Audio%20Component%20Object%20References) (BurstCameraShakeInstance :type Camera%20Shake%20Base%20Object%20Reference) (BurstDecalInstance :type Decal%20Actor%20Object%20Reference))
  (bind _output (Utilities|Array|Get(acopy) ParticleComponents))
  (bind _asfort_player_pawn_athena (Utilities|Casting|CastToFortPlayerPawnAthena MyTarget))
  (Development|PrintString 0 (Utilities|String|ToString(Float) (Math|Float|MapRangeClamped (Variables|/Script/SunRoseFlyingRuntime.FortPawnComponent_WingedFlying|GetNativeCurrentHeightAboveWorld (Actor|GetComponentbyClass _asfort_player_pawn_athena "/SunRoseFlyingGameplay/Items/Wings/Blueprints/FlyingPawnComponent.FlyingPawnComponent_C")) 2.5e+03 5e+03 0.0 1.0)) false false "(R=1.000000,G=0.329981,B=0.000000,A=1.000000)" 1e+01)
  (Effects|Components|ParticleSystem|SetBoolParameter _output "IsLocalPlayer" (Pawn|IsLocallyControlled _asfort_player_pawn_athena))
  (Effects|Components|ParticleSystem|SetFloatParameter _output "AltitudeInterpolate" (Math|Float|MapRangeClamped (Variables|/Script/SunRoseFlyingRuntime.FortPawnComponent_WingedFlying|GetNativeCurrentHeightAboveWorld (Actor|GetComponentbyClass _asfort_player_pawn_athena "/SunRoseFlyingGameplay/Items/Wings/Blueprints/FlyingPawnComponent.FlyingPawnComponent_C")) 2.5e+03 5e+03 0.0 1.0))
  (return))
