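; One-shot gameplay cue that plays the boost effect for the SunRose wings in the flying gameplay.
; On burst, it passes whether the target player pawn is locally controlled to a spawned particle component as "IsLocalPlayer".
; It then casts the pawn's current weapon to the wings Blueprint and drives its dynamic material instance.
; The material receives the current game time as a "Boost" timestamp, along with a fixed boost speed and orange color.
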(fn OnBurst(EitherVFX-preferthisversion)
  (params (MyTarget :type Actor%20Object%20Reference) (Parameters :type Gameplay%20Cue%20Parameters%20Structure%20%28by%20ref%29) (ParticleComponents :type Array%20of%20FXSystem%20Component%20Object%20References) (AudioComponents :type Array%20of%20Audio%20Component%20Object%20References) (BurstCameraShakeInstance :type Camera%20Shake%20Base%20Object%20Reference) (BurstDecalInstance :type Decal%20Actor%20Object%20Reference))
  (bind _asfort_player_pawn_athena (Utilities|Casting|CastToFortPlayerPawnAthena MyTarget))
  (Effects|Components|ParticleSystem|SetBoolParameter (Utilities|Array|Get(acopy) ParticleComponents) "IsLocalPlayer" (Pawn|IsLocallyControlled _asfort_player_pawn_athena))
  (bind _asb_athena_sun_rose_wings (Utilities|Casting|CastToB_Athena_SunRose_Wings (Variables|/Script/FortniteGame.FortPawn|GetCurrentWeapon _asfort_player_pawn_athena)))
  (Rendering|Material|SetScalarParameterValue (Variables|/SunRoseFlyingGameplay/Items/Wings/Blueprints/B_Athena_SunRose_Wings.B_Athena_SunRose_Wings_C|GetMaterialInstance _asb_athena_sun_rose_wings) "Boost" (Utilities|Time|GetGameTimeinSeconds _asfort_player_pawn_athena) :!class /Script/Engine.MaterialInstanceDynamic)
  (Rendering|Material|SetScalarParameterValue (Variables|/SunRoseFlyingGameplay/Items/Wings/Blueprints/B_Athena_SunRose_Wings.B_Athena_SunRose_Wings_C|GetMaterialInstance _asb_athena_sun_rose_wings) "BoostSpeed" 1.6 :!class /Script/Engine.MaterialInstanceDynamic)
  (Rendering|Material|SetVectorParameterValue (Variables|/SunRoseFlyingGameplay/Items/Wings/Blueprints/B_Athena_SunRose_Wings.B_Athena_SunRose_Wings_C|GetMaterialInstance _asb_athena_sun_rose_wings) "BoostColor" "(R=1.000000,G=0.373878,B=0.100721,A=1.000000)" :!class /Script/Engine.MaterialInstanceDynamic)
  (return))
