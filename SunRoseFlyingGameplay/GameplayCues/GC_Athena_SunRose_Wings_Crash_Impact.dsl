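; Burst gameplay cue for a SunRose wings crash impact, adding effects and sound at the hit location.
; Shows a sound indicator for players off the local team.
; A water-body causer gets depth-aware water FX, radial force, lens droplets and zeroed camera shake.
; Otherwise it spawns landing Niagara, plus a fading decal and radial force except on train cars and low-spec platforms; both branches set the sound's Water parameter.
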(fn OnBurst(EitherVFX-preferthisversion)
  (params (MyTarget :type Actor%20Object%20Reference) (Parameters :type Gameplay%20Cue%20Parameters%20Structure%20%28by%20ref%29) (ParticleComponents :type Array%20of%20FXSystem%20Component%20Object%20References) (AudioComponents :type Array%20of%20Audio%20Component%20Object%20References) (BurstCameraShakeInstance :type Camera%20Shake%20Base%20Object%20Reference) (BurstDecalInstance :type Decal%20Actor%20Object%20Reference))
  (bind (_normalizedmagnitude _rawmagnitude _effectcontext _matchedtagname _originaltag _aggregatedsourcetags _aggregatedtargettags _location _normal _instigator _effectcauser) (Ability|GameplayCue|BreakGameplayCueParameters Parameters))
  (bind _impactnormal (LocalVariables|GetImpactNormal))
  (bind _returnvalue (* _impactnormal 4.0))
  (bind _impactlocation (LocalVariables|GetImpactLocation))
  (bind _returnvalue_2 (+ _impactlocation _returnvalue))
  (|CallParentFunction MyTarget Parameters ParticleComponents AudioComponents BurstCameraShakeInstance BurstDecalInstance)
  (LocalVariables|SetImpactLocation _location)
  (LocalVariables|SetImpactNormal _normal)
  (bind _asfort_player_pawn (Utilities|Casting|CastToFortPlayerPawn MyTarget)
    (:then
      (LocalVariables|SetPlayerPawn _asfort_player_pawn)
      (if (not (Team|OnLocalPlayersTeam (LocalVariables|GetPlayerPawn) (LocalVariables|GetPlayerPawn)))
        (Fortnite|ShowSoundIndicatorLocal :WorldContextObject (LocalVariables|GetPlayerPawn) :TrackedActor (LocalVariables|GetPlayerPawn) :Location (Transformation|GetActorLocation (LocalVariables|GetPlayerPawn)) :MaxAudibleDistance (ScalableFloat|GetValueatLevel (Variables|Default|GetRow_AVIndicatorRange)) :IndicatorType "Gunshot" :Instigator (LocalVariables|GetPlayerPawn) :Tint "(R=0.000000,G=0.000000,B=0.000000,A=0.000000)" :OverrideIcon (Variables|Default|GetAVIcon) :AttenuationMaxRangeMultiplier 1.0 :AttenuationIntensityMultiplier 1.0)))
    (:CastFailed
      _))
  (bind _asfort_water_body_actor (Utilities|Casting|CastToFortWaterBodyActor _effectcauser)
    (:then
      (bind _outwatersurfaceinfo (WaterBody|GetWaterSurfaceInfoStruct _asfort_water_body_actor (LocalVariables|GetImpactLocation) true false false))
      (bind _returnvalue_3 (Effects|Components|ParticleSystem|SpawnFXAtLocation (Variables|GetMyTarget) (Variables|Default|GetImpactFX_Water) (Utilities|Struct|BreakWaterSurfaceInfo _outwatersurfaceinfo) (Math|Rotator|MakeRotfromX (.WaterPlaneNormal (Utilities|Struct|BreakWaterSurfaceInfo _outwatersurfaceinfo)))))
      (Game|SpawnActorfromClass (Variables|GetMyTarget) "/Game/Effects/Niagara/Blueprints/PhsyicsFields/FN_RadialForce.FN_RadialForce_C" (Math|Transform|MakeTransform (LocalVariables|GetImpactLocation)) "Undefined" "MultiplyWithRoot" 0 3.0 0.75 8e+02 1.5 0.05 0)
      (Effects|Components|ParticleSystem|SetFloatParameter _returnvalue_3 "WaterDepth" (.WaterDepth (Utilities|Struct|BreakWaterSurfaceInfo _outwatersurfaceinfo)))
      (Effects|Components|ParticleSystem|SetFloatParameter _returnvalue_3 "WaterBodyIndex" (Math|Conversions|ToFloat(Integer) (.WaterBodyIdx (Utilities|Struct|BreakWaterSurfaceInfo _outwatersurfaceinfo))))
      (CallFunction|SpawnAudioandSetParameters (Variables|GetMyTarget) true (LocalVariables|GetImpactLocation))
      (Variables|/Script/Engine.CameraShakeBase|SetShakeScale 0.0 BurstCameraShakeInstance)
      (Feedback|SpawnGenericCameraLensEffect (Utilities|Casting|CastToFortPlayerPawn (Variables|GetMyTarget)) "/Game/Effects/Fort_Effects/Camera_Effects/B_CameraLens_Splash_Water_Droplets.B_CameraLens_Splash_Water_Droplets_C")
      (return))
    (:CastFailed
      (Effects|Components|ParticleSystem|SpawnFXAtLocation (Variables|GetMyTarget) "/SunRoseFlyingGameplay/Items/Wings/FX/NS_SunRose_Wings_LandingImpact.NS_SunRose_Wings_LandingImpact" _returnvalue_2 (Math|Conversions|RotationFromXVector (LocalVariables|GetImpactNormal)) "1.000000,1.000000,1.000000" true "AutoRelease")
      (bind _returnvalue_4 (GameplayTags|GetOwnedGameplayTags _effectcauser :!class /Script/GameplayTags.GameplayTagAssetInterface))
      (if (not (Math|Boolean|ORBoolean (Utilities|Array|ContainsItem (GameplayTags|BreakGameplayTagContainer _returnvalue_4) "(TagName=\"Athena.Gameplay.Train.Car\")") (Fortnite|IsMobilePlatform) (Class|PlatformHelpersLibrary|IsPcPerfMode (Variables|GetMyTarget)) (TouchPlatform|IsMobileGame (Variables|GetMyTarget)) (TouchPlatform|IsNintendoSwitchorNintendoSwitchPreview)))
        (bind _returnvalue_5 (Rendering|Decal|SpawnDecalatLocation (Variables|GetMyTarget) (Variables|Default|GetDecal) (Variables|Default|GetDecalSize) _returnvalue_2 (Math|Conversions|RotationFromXVector (* (LocalVariables|GetImpactNormal) -1.0)) (Variables|Default|GetDecal_Lifespan)))
        (Rendering|Components|Decal|SetFadeOut _returnvalue_5 (Variables|Default|GetDecal_Fadeout_StartDelay) (Variables|Default|GetDecal_Fadeout_Duration) true)
        (Game|SpawnActorfromClass (Variables|GetMyTarget) "/Game/Effects/Niagara/Blueprints/PhsyicsFields/FN_RadialForce.FN_RadialForce_C" (Math|Transform|MakeTransform (LocalVariables|GetImpactLocation)) "Undefined" "MultiplyWithRoot" 0 3.0 0.75 1e+03 1.5 0.05 0))
      (CallFunction|SpawnAudioandSetParameters (Variables|GetMyTarget) false (LocalVariables|GetImpactLocation))
      (return))))

(fn SpawnAudioAndSetParameters
  (params (MyTarget :type Object%20Reference) (isWater :type Boolean) (ImpactLocation :type Vector))
  (bind _returnvalue (Audio|SpawnSoundatLocation MyTarget (Variables|Default|GetImpactSFX) (Variables|GetImpactLocation)))
  (LocalVariables|SetAudioComponent _returnvalue)
  (Audio|Parameter|SetBooleanParameter (LocalVariables|GetAudioComponent) "Water" (Variables|GetisWater) :!class /Script/AudioExtensions.AudioParameterControllerInterface))
