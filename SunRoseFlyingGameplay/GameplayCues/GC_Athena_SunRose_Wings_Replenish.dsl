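; Burst gameplay cue that plays feedback when a player's SunRose wings replenish.
; It casts the target to the Fortnite player pawn and triggers an anim input event that the player AnimBP uses for an additive wing animation.
; It spawns an auto-released Niagara effect attached to the current weapon's mesh.
; If that weapon is the wings, it writes current game time into the material's Replenish parameter, timestamping the shader effect.
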(fn OnBurst(EitherVFX-preferthisversion)
  (params (MyTarget :type Actor%20Object%20Reference) (Parameters :type Gameplay%20Cue%20Parameters%20Structure%20%28by%20ref%29) (ParticleComponents :type Array%20of%20FXSystem%20Component%20Object%20References) (AudioComponents :type Array%20of%20Audio%20Component%20Object%20References) (BurstCameraShakeInstance :type Camera%20Shake%20Base%20Object%20Reference) (BurstDecalInstance :type Decal%20Actor%20Object%20Reference))
  (bind _asfort_player_pawn_athena (Utilities|Casting|CastToFortPlayerPawnAthena MyTarget))
  ;[1 1 1] Used On "SunRose_Player_AnimBP" To Play Additive Wing Animation
  (bind _returnvalue (Class|FortAnimInputEvent|CreateAnimInputEvent "aie.WingReplenish"))
  (Animation|TriggerAnimInputEvent _asfort_player_pawn_athena _returnvalue)
  (Effects|SpawnFXSystemAttached :FXSystemAsset "/SunRoseFlyingGameplay/Items/Wings/FX/NS_SunRose_Wings_Replenish.NS_SunRose_Wings_Replenish" :AttachToComponent (Variables|/Script/FortniteGame.FortWeapon|GetWeaponMesh (Variables|/Script/FortniteGame.FortPawn|GetCurrentWeapon _asfort_player_pawn_athena)) :AttachPointName "None" :Scale "1.000000,1.000000,1.000000" :LocationType "KeepRelativeOffset" :bAutoDestroy true :PoolingMethod "AutoRelease" :bAutoActivate true)
  (bind _asb_athena_sun_rose_wings (Utilities|Casting|CastToB_Athena_SunRose_Wings (Variables|/Script/FortniteGame.FortPawn|GetCurrentWeapon _asfort_player_pawn_athena))
    (:then
      (Rendering|Material|SetScalarParameterValue (Variables|/SunRoseFlyingGameplay/Items/Wings/Blueprints/B_Athena_SunRose_Wings.B_Athena_SunRose_Wings_C|GetMaterialInstance _asb_athena_sun_rose_wings) "Replenish" (Utilities|Time|GetGameTimeinSeconds _asfort_player_pawn_athena) :!class /Script/Engine.MaterialInstanceDynamic))
    (:CastFailed
      _))
  (return))
